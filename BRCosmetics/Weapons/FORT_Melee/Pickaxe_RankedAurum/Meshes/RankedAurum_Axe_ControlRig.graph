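# Control Rig for the RankedAurum pickaxe that procedurally adjusts its dangling charm bones.
# Two Spherical Pose Readers measure dyn_charm_00's downward axis relative to the weapon handle, using roughly opposite Y rotation offsets and flipped width scaling for the second.
# Each reader's output weights a Modify Transforms step that moves and rotates the left and right charm bones as the parent charm swings into that region.
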ControlRigBlueprint /BRCosmetics/Weapons/FORT_Melee/Pickaxe_RankedAurum/Meshes/RankedAurum_Axe_ControlRig
graphs [
 RigVMFunctionLibrary {links [], nodes []}
 RigVMModel {nodes [
   RigUnit_BeginExecution = &RigUnit_BeginExecution()
   n1 = &RigUnit_SphericalPoseReader(
    ActiveRegionScaleFactors=(PositiveWidth=2,NegativeWidth=0.3,PositiveHeight=2,NegativeHeight=2)
    ActiveRegionSize=0.3
    Debug=(bDrawDebug=false,bDraw2D=False,bDrawLocalAxes=False,DebugScale=25,DebugSegments=20,DebugThickness=0.25)
    DriverAxis=(X=0,Y=0,Z=-1)
    DriverItem=(Type=Bone,Name="dyn_charm_00")
    FalloffRegionScaleFactors=(PositiveWidth=-0.5,NegativeWidth=0.6,PositiveHeight=-1,NegativeHeight=-1)
    FalloffSize=0.5
    FlipHeightScaling=False
    FlipWidthScaling=False
    OptionalParentItem=(Type=Bone,Name="Melee_TwoHanded_Handle")
    RotationOffset=(X=0,Y=120,Z=0)
    ExecutePin=@RigUnit_BeginExecution.ExecutePin
   )
   > &RigUnit_ModifyTransforms(
    ItemToModify="((Item=(Type=Bone,Name=\"dyn_charm_l\"),Transform=(Rotation=(X=0,Y=0.1297,Z=0,W=0.9916),Translation=(X=2.43,Y=0.0004,Z=-0.3214),Scale3D=(X=1,Y=1,Z=1))),(Item=(Type=Bone,Name=\"dyn_charm_r\"),Transform=(Rotation=(X=0,Y=0.0189,Z=0,W=0.9998),Translation=(X=-1.439 ...(truncated)"
    WeightMinimum=0
    Weight=@n1.OutputParam
   )
   > n2 = &RigUnit_SphericalPoseReader(
    ActiveRegionScaleFactors=(PositiveWidth=2,NegativeWidth=0.3,PositiveHeight=2,NegativeHeight=2)
    ActiveRegionSize=0.3
    Debug=(bDrawDebug=false,bDraw2D=False,bDrawLocalAxes=False,DebugScale=25,DebugSegments=20,DebugThickness=0.25)
    DriverAxis=(X=0,Y=0,Z=-1)
    DriverItem=(Type=Bone,Name="dyn_charm_00")
    FalloffRegionScaleFactors=(PositiveWidth=-0.5,NegativeWidth=0.7,PositiveHeight=-1,NegativeHeight=-1)
    FalloffSize=0.5
    FlipHeightScaling=False
    FlipWidthScaling=true
    OptionalParentItem=(Type=Bone,Name="Melee_TwoHanded_Handle")
    RotationOffset=(X=0,Y=-115,Z=0)
   )
   > &RigUnit_ModifyTransforms(
    ItemToModify="((Item=(Type=Bone,Name=\"dyn_charm_l\"),Transform=(Rotation=(X=0,Y=-0.0801,Z=0,W=0.9968),Translation=(X=1.0768,Y=-0.0059,Z=0.0575),Scale3D=(X=1,Y=1,Z=1))),(Item=(Type=Bone,Name=\"dyn_charm_r\"),Transform=(Rotation=(X=0,Y=-0.2856,Z=0,W=0.9583),Translation=(X=-1 ...(truncated)"
    WeightMinimum=0
    Weight=@n2.OutputParam
   )
  ]}
]
hierarchy [
 b_AR_Root {loc=<0 0 0>, rot=<0 0 0.7071 0.7071>}
 Melee_TwoHanded_Handle {parent=b_AR_Root, loc=<8.1477 -0.008 -29.0793>, rot=<0 0 0.7071 -0.7071>}
 dyn_charm_00 {parent=Melee_TwoHanded_Handle, loc=<0.7981 -0.9895 -43.2718>, rot=<0 0 0 1>}
 dyn_charm_01 {parent=dyn_charm_00, loc=<0 0 -3.267>, rot=<0 0 0 1>}
 dyn_charm_l {parent=dyn_charm_00, loc=<1.2029 0 -1.9607>, rot=<0 0 0 1>}
 dyn_charm_r {parent=dyn_charm_00, loc=<-1.203 0 -1.9607>, rot=<0 0 0 1>}
]
