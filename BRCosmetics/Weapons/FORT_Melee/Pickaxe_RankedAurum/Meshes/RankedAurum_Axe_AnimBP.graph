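# Animation Blueprint for the RankedAurum pickaxe that adds physics swing to its two dangling charm bones.
# The cached input pose feeds two chains of Anim Dynamics nodes on dyn_charm_00 and dyn_charm_01.
# A bool blend on IsInFrontEnd picks world-space simulation with angular springs in the front end, or looser actor-space simulation otherwise.
# The Control Rig node is unconnected, and the event graph has no logic.
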AnimBlueprint /BRCosmetics/Weapons/FORT_Melee/Pickaxe_RankedAurum/Meshes/RankedAurum_Axe_AnimBP
target_skeleton=&/RankedAurum_Axe_Skeleton
graphs [
 AnimGraph {nodes [
   Control Rig() {
    alphaBoolBlend {blendOption=Linear}
    controlRigAssetReference {blueprintRigClass=&/RankedAurum_Axe_ControlRig.RankedAurum_Axe_ControlRig_C}
   }
   n1 = Anim Dynamics - Bone: dyn_charm_00(ComponentPose=Local To Component(LocalPose=Use cached pose 'PostInput'()), Alpha=1, ExternalForce="0,0,0") {
    angularDampingOverride=0.97
    angularSpringConstant=1000
    bAngularSpring=true
    bOverrideAngularDamping=true
    bUseGravityOverride=true
    simulationSpace=World
    alphaBoolBlend {blendOption=Linear}
    boundBone {boneName=dyn_charm_00}
    componentAppliedLinearAccClamp=<1000 1000 1500>
    componentLinearAccScale=<0.4 0.4 0.4>
    componentLinearVelScale=<0.5 0.5 0.5>
    gravityOverride {z=-980}
    physicsBodyDefinitions [{
      boundBone {boneName=dyn_charm_00}
      boxExtents=<9 9 9>
      constraintSetup {twistAxis=AxisZ, angularLimitsMax=<50 20 10>, angularLimitsMin=<-50 -20 -10>, angularTarget=<0 0 0>}
      localJointOffset=<0 0 -1>
     }]
   }
   n2 = Anim Dynamics - Bone: dyn_charm_01(ComponentPose=@n1, Alpha=1, ExternalForce="0,0,0") {
    angularDampingOverride=0.97
    angularSpringConstant=1000
    bAngularSpring=true
    bOverrideAngularDamping=true
    bUseGravityOverride=true
    simulationSpace=World
    alphaBoolBlend {blendOption=Linear}
    boundBone {boneName=dyn_charm_01}
    componentAppliedLinearAccClamp=<1000 1000 1500>
    componentLinearAccScale=<0.3 0.3 0.3>
    componentLinearVelScale=<0.5 0.5 0.5>
    gravityOverride {z=-980}
    physicsBodyDefinitions [{
      boundBone {boneName=dyn_charm_01}
      boxExtents=<9 9 9>
      constraintSetup {twistAxis=AxisZ, angularLimitsMax=<50 10 10>, angularLimitsMin=<-50 -10 -10>, angularTarget=<0 0 0>}
      localJointOffset=<0 0 -1>
     }]
   }
   n3 = Anim Dynamics - Bone: dyn_charm_00(ComponentPose=Local To Component(LocalPose=Use cached pose 'PostInput'()), Alpha=1, ExternalForce="0,0,0") {
    angularDampingOverride=0.97
    bOverrideAngularDamping=true
    bUseGravityOverride=true
    simulationSpace=Actor
    alphaBoolBlend {blendOption=Linear}
    boundBone {boneName=dyn_charm_00}
    componentAppliedLinearAccClamp=<1000 1000 1500>
    componentLinearAccScale=<0.3 0.3 0.3>
    componentLinearVelScale=<0.5 0.5 0.5>
    gravityOverride {z=-980}
    physicsBodyDefinitions [{
      boundBone {boneName=dyn_charm_00}
      boxExtents=<7 7 8>
      constraintSetup {twistAxis=AxisZ, angularLimitsMax=<50 50 10>, angularLimitsMin=<-50 -50 -10>, angularTarget=<0 0 0>}
      localJointOffset=<0 0 -1>
     }]
   }
   n4 = Anim Dynamics - Bone: dyn_charm_01(ComponentPose=@n3, Alpha=1, ExternalForce="0,0,0") {
    angularDampingOverride=0.97
    bOverrideAngularDamping=true
    bUseGravityOverride=true
    simulationSpace=Actor
    alphaBoolBlend {blendOption=Linear}
    boundBone {boneName=dyn_charm_01}
    componentAppliedLinearAccClamp=<1000 1000 1500>
    componentLinearAccScale=<0.3 0.3 0.3>
    componentLinearVelScale=<0.5 0.5 0.5>
    gravityOverride {z=-980}
    physicsBodyDefinitions [{
      boundBone {boneName=dyn_charm_01}
      boxExtents=<7 7 8>
      constraintSetup {twistAxis=AxisZ, angularLimitsMax=<50 30 10>, angularLimitsMin=<-50 -30 -10>, angularTarget=<0 0 0>}
      localJointOffset=<0 0 -1>
     }]
   }
   Output Pose(Result=Blend Poses by bool(BlendPose_0=Component To Local(ComponentPose=@n2), BlendPose_1=Component To Local(ComponentPose=@n4), bActiveValue=Get bIsInFrontEnd()) {blendType=Linear}) {layerGroup=None}
   AnimGraphNode_SubInput_0 = Input Pose()
   Save cached pose 'PostInput'(Pose=@AnimGraphNode_SubInput_0) {cacheName=PostInput}
  ]}
 EventGraph {nodes [
   TryGetPawnOwner()
   Event BlueprintUpdateAnimation()
  ]}
]
variables []
