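# Physics asset for the RankedAurum pickaxe mesh, a cosmetic two-handed melee weapon.
# It has a single body on the Melee_TwoHanded_Handle bone, built from one large capsule that is offset and rotated relative to that bone.
# There are no constraints, so the asset defines one simple collision volume with no jointed parts.
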PhysicsAsset /BRCosmetics/Weapons/FORT_Melee/Pickaxe_RankedAurum/Meshes/RankedAurum_Axe_PhysicsAsset
bodies {Melee_TwoHanded_Handle [{shapeType=Capsule, length=71.7192, radius=45.2147, center=<0.8015 -1.0054 14.9517>, rotation=<rot 0.0057 -90 -0.0032>}]}
constraints []
